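# Material function for the Basil pickaxe that composites a spray image onto the incoming material attributes.
# Parameterized offset and center-based scale position the spray, clipped by its alpha, the 0-1 UV mask and the inverted red channel of a mask texture.
# The mask's green channel dirties the spray area, tinting base color toward a dirt color and raising roughness.
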MaterialFunction /BRCosmetics/Weapons/FORT_Melee/Pickaxe_Basil/Materials/MF_Basil_AlignSpray
graph {nodes [
  n1 = GetMaterialAttributes(None=FunctionInput() {inputType=FunctionInput_MaterialAttributes}) {attributeGetTypes [#1, #2]}
  n2 = TextureCoordinate()
  n3 = VectorParameter() {parameterName=SprayUV_Offset, defaultValue=<rgba -0.14 0.16 0 1>}
  n4 = ScalarParameter() {defaultValue=0.33, parameterName=Spray Scale, sliderMin=4}
  n5 = MaterialFunctionCall(UVs=Add(A=@n2, B=ComponentMask(None=@n3.RGB) {g=true, r=true}), Texture Scale=@n4) {materialFunction=&ScaleUVsByCenter}
  n6 = TextureSampleParameter2D(UVs=@n5.UVs) {parameterName=Spray, texture=/Game/2dAssets/Sprays/Season5/Textures/T-Sprays-Boogie}
  n7 = TextureSampleParameter2D() {parameterName=Spray mask, samplerType=SAMPLERTYPE_Masks, texture=&T_BlackMask}
  n8 = Multiply(A=Multiply(A=@n5.0-1 mask, B=@n6.A), B=OneMinus(None=@n7.R))
  n9 = LinearInterpolate(A=@n1.BaseColor, B=@n6.RGB, Alpha=@n8)
  n10 = VectorParameter() {parameterName=Dirt Color, defaultValue=<rgba 0.02 0.0096 0.0079 1>}
  n11 = Multiply(A=@n8, B=@n7.G)
  n12 = Multiply(A=@n11) {constB=0.8}
  FunctionOutput(None=SetMaterialAttributes!(
    MaterialAttributes=@n1.MaterialAttributes
    BaseColor=LinearInterpolate(A=@n9, B=@n10.RGB, Alpha=@n12)
    Roughness=LinearInterpolate(A=@n1.Roughness, Alpha=@n11)
   ) {attributeSetTypes [#1, #2]})
  Multiply(A=@n2, B=@n4)
 ]}
