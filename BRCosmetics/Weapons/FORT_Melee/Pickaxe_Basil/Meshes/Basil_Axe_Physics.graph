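# Physics asset for the Basil two-handed pickaxe, giving collision bodies and joints to its handle and dangling "dyn" bones.
# A constraint chain runs from the handle capsule through the dyn_base sphere to the dyn_base1 capsule, which branches into large left and right block capsules.
# The joints below dyn_base have free swing and twist, so both blocks can rotate loosely; only the handle-to-base joint is not set free.
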PhysicsAsset /BRCosmetics/Weapons/FORT_Melee/Pickaxe_Basil/Meshes/Basil_Axe_Physics
bodies {
 Melee_TwoHanded_Handle [{shapeType=Capsule, length=35.2218, radius=6.8734, center=<0.7642 -2.5226 -7.5804>, rotation=<rot 0.7437 -90 0.0309>}]
 dyn_base [{shapeType=Sphere, radius=5, shapeName=dyn_base_sphere, center=<16.6252 -0.8219 0.4971>}]
 dyn_base1 [{shapeType=Capsule, length=15, radius=5, shapeName=dyn_base1_capsule, rotation=<rot -42 0 0>}]
 dyn_l_block [{shapeType=Capsule, length=24.8459, radius=19.2937, center=<-18.5374 5.2778 1.1993>, rotation=<rot 0 -104.0344 -57.8164>}]
 dyn_r_block [{shapeType=Capsule, length=29.1611, radius=19.517, center=<20.5736 -0.5498 -0.8004>, rotation=<rot 0 -93.6979 -55.9595>}]
}
constraints [
 {bone1Name=dyn_base, bone2Name=Melee_TwoHanded_Handle, swing1LimitDegrees=45, swing2LimitDegrees=45, twistLimitDegrees=45}
 {
  bone1Name=dyn_l_block
  bone2Name=dyn_base1
  swing1LimitDegrees=45
  swing1Motion=Free
  swing2LimitDegrees=45
  swing2Motion=Free
  twistLimitDegrees=45
  twistMotion=Free
 }
 {
  bone1Name=dyn_r_block
  bone2Name=dyn_base1
  swing1LimitDegrees=45
  swing1Motion=Free
  swing2LimitDegrees=45
  swing2Motion=Free
  twistLimitDegrees=45
  twistMotion=Free
 }
 {
  bone1Name=dyn_base1
  bone2Name=dyn_base
  swing1LimitDegrees=45
  swing1Motion=Free
  swing2LimitDegrees=45
  swing2Motion=Free
  twistLimitDegrees=45
  twistMotion=Free
 }
]
