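# Animation Blueprint for the Basil pickaxe that adds secondary motion to its dynamic bones.
# The AnimGraph uses RigidBody physics when bRigidBodyEnabled is true, otherwise a spring-and-trail chain on dyn_r_block.
# Both paths additively rotate dyn_base1 by a falling value mapped from pawn vertical velocity, softened without rigid body.
# The EventGraph also eases RigidBody alpha by jump state and scales it down in the front end.
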AnimBlueprint /BRCosmetics/Weapons/FORT_Melee/Pickaxe_Basil/Meshes/Basil_Axe_Skeleton_AnimBlueprint
target_skeleton=&/Basil_Axe_Skeleton
graphs [
 AnimGraph {nodes [
   n1 = Spring controller - Bone: dyn_r_block!(ComponentPose:Component Space Pose Link Structure=Local To Component(LocalPose=Input Pose() {name=InPose_0}), Alpha:float=0.9) -> (Pose:Component Space Pose Link Structure) {
    is=title AnimGraphNode_SpringBone
    springDamping=17
    springStiffness=500
    alphaBoolBlend {blendOption=Linear}
    alphaScaleBiasClamp {clampMax=0.05, scale=0.07}
    springBone {boneName=dyn_r_block}
   }
   n2 = Trail controller - Bone: dyn_r_block(ComponentPose=@n1, Alpha=1) {
    relaxationSpeedScale=10
    fakeVelocity=<0.2 0.2 0.2>
    rotationLimits [{limitMax=<180 180 180>}, {limitMax=<180 180 180>}]
    rotationOffsets [<0 0 0>, <0 0 0>]
    trailBone {boneName=dyn_r_block}
   }
   n3 = Get falling() -> (falling:double)
   n4 = "Transform (Modify) Bone - Bone: dyn_base1"!(
    ComponentPose=@n2
    bAlphaBoolEnabled:bool=True
    Translation="0,0,0"
    Rotation_Roll:float=@n3
    Rotation_Pitch:float=0
    Rotation_Yaw:float=@n3
    Scale="1,1,1"
   ) -> (Pose) {
    alphaInputType=Bool
    rotationMode=BMM_Additive
    rotationSpace=BCS_BoneSpace
    alphaBoolBlend {blendInTime=0.4, blendOption=ExpInOut, blendOutTime=0.5}
    boneToModify {boneName=dyn_base1}
    rotation {roll=5, yaw=5}
   }
   n5 = RigidBody(ComponentPose=Local To Component(LocalPose=Input Pose()), Alpha=Get alpha_jump() -> (alpha_jump:double)) {
    bClampLinearTranslationLimitToRefPose=true
    overridePhysicsAsset=&/Basil_Axe_Physics
    alphaBoolBlend {blendOption=Linear}
    componentAppliedLinearAccClamp=<3000 3000 3000>
    componentLinearAccScale=<0.4 0.2 0.2>
    componentLinearVelScale=<0.4 0.2 0.2>
   }
   n6 = Get falling()
   n7 = "Transform (Modify) Bone - Bone: dyn_base1"(ComponentPose=@n5, bAlphaBoolEnabled=True, Translation="0,0,0", Rotation_Roll=@n6, Rotation_Pitch=0, Rotation_Yaw=@n6, Scale="1,1,1") {
    alphaInputType=Bool
    rotationMode=BMM_Additive
    rotationSpace=BCS_BoneSpace
    alphaBoolBlend {blendOption=Linear}
    boneToModify {boneName=dyn_base1}
   }
   Output Pose(Result=Blend Poses by bool(BlendPose_0=Component To Local(ComponentPose=@n7), BlendPose_1=Component To Local(ComponentPose=@n4), bActiveValue=Get bRigidBodyEnabled()) {blendType=Linear}) {layerGroup=None}
  ]}
 EventGraph {nodes [
   n1 = Event BlueprintUpdateAnimation()
   > Set Delta Time X(Delta Time X=@n1.DeltaTimeX)
   > n2 = Sequence!(execute) -> (then_0, then_1, then_2:Exec, then_3:Exec, then_4:Exec, then_5:Exec)
   > n3 = Branch(Condition=Get bIsJumping())
   > Set alpha_jump(alpha_jump:double=FInterpTo(Current=Get alpha_jump(), Target=0.7, DeltaTime=Get Delta Time X(), InterpSpeed=2)) -> (Output_Get:double)
   Set alpha_jump(execute=@n3.else, alpha_jump=FInterpTo(Current=Get alpha_jump(), Target=1, DeltaTime=Get Delta Time X(), InterpSpeed=2))
   Branch(execute=@n2.then_2, Condition=Get bIsInFrontEnd())
   > Set alpha_jump(alpha_jump=float * float(A=Get alpha_jump(), B=0.85) {is=title K2Node_PromotableOperator})
   n4 = Branch(execute=@n2.then_1, Condition=Get bRigidBodyEnabled())
   n5 = MapRangeClamped(Value=Get PawnVelocityZ(), InRangeA=630, InRangeB=-630, OutRangeA=5, OutRangeB=-5)
   > Set falling(falling:double=@n5) -> (Output_Get:double)
   Set falling(execute=@n4.else, falling=float / float(A=@n5, B=1.5))
   TryGetPawnOwner()
  ]}
 OnChange {nodes [
   n1 = Inputs!() -> (Exec:Exec, OnChange:bool=~)
   n2 = Local Boolean!() -> (Variable:bool=~) {is=title K2Node_TemporaryVariable}
   > Branch(Condition="Not Equal (Boolean)"(A=@n1, B=@n2) -> (ReturnValue=~))
   > Assign!(execute:Exec, Variable:bool=@n2, Value:bool=@n1) -> (then:Exec) {is=title K2Node_AssignmentStatement}
   > Outputs!(Out:Exec)
  ]}
]
variables [
 alpha_jump {type=number}
 alpha_scalability1 {type=number}
 alpha_scalability2 {type=number}
 delayed jumping {type=boolean}
 Delta Time X {type=number}
 falling {type=number}
 falling_scalability {type=number}
 NewVar {type=boolean}
]
